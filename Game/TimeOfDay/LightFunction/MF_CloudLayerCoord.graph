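# Material function that outputs scrolling world-space UVs and a mip level for a cloud layer in time-of-day light functions.
# UVs project world XY scaled by Size and pan along V at Speed, with the safe-zone cloud collection's CustomCloudsTime replacing Time when positive.
# MIP is log2 of the pixel's world footprint, from FOV, view size and depth, over Size times Resolution.
# A static switch shrinks Size for material-editor preview.
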MaterialFunction /Game/TimeOfDay/LightFunction/MF_CloudLayerCoord
graph {nodes [
  n1 = ComponentMask(None=WorldPosition().XYZ) {g=true, r=true}
  n2 = FunctionInput() {bUsePreviewValueAsDefault=true, inputName=Size, inputType=FunctionInput_Scalar, previewValue=<5000 0 0 1>}
  n3 = StaticSwitch(
   True=Divide(A=@n2, B=Constant() {r=10})
   False=@n2
   Value=FunctionInput(Preview=StaticBool()) {bUsePreviewValueAsDefault=true, inputName=PreviewInMatEditor, inputType=FunctionInput_StaticBool}
  )
  n4 = CollectionParameter() {collection=/Game/Athena/SafeZone/MPC_SafeZoneClouds, parameterName=CustomCloudsTime}
  n5 = LinearInterpolate(A=Time(), B=@n4, Alpha=Ceil(None=Saturate(None=@n4)))
  n6 = FunctionInput() {bUsePreviewValueAsDefault=true, inputName=Speed, inputType=FunctionInput_Scalar, sortPriority=1, previewValue=<1 0 0 1>}
  FunctionOutput(None=Panner(Coordinate=Divide(A=@n1, B=@n3), Time=Multiply(A=@n5, B=@n6)) {speedY=0.01}) {outputName=UVs, sortPriority=1}
  n7 = ComponentMask(None=Multiply(A=ViewProperty() {property=MEVP_TanHalfFieldOfView}.Property, B=ViewProperty() {property=MEVP_ViewSize}.InvProperty)) {r=true}
  n8 = Divide(A=Multiply(A=Multiply(A=@n7) {constB=2}, B=PixelDepth().R), B=@n3)
  n9 = FunctionInput() {inputName=Resolution, inputType=FunctionInput_Vector2, sortPriority=10}
  FunctionOutput(None=Logarithm2!(X=Multiply(A=@n8, B=@n9))) {outputName=MIP, sortPriority=2}
 ]}
