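# Light function material that casts moving cloud shadows, with extra darkening under the storm cloud.
# Two MF_CloudLayerCoord layers sample T_CloudNoise_Artemis at different sizes and speeds, cross-fading by camera distance with a noise-modulated blend range.
# The remapped noise multiplies a storm term: an XY sphere mask around StormCloudLocation, gated by isStormActive, both parameters from FortniteMaterialParameters.
# A camera-height ship shadow fade is computed but feeds nothing visible.
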Material /Game/TimeOfDay/LightFunction/LF_CloudShadows_Artemis
graph {
 properties {materialDomain=MD_LightFunction, pixelDepthOffsetMode=PDOM_Legacy, shadingModel=MSM_Unlit, usedShadingModels=Unlit}
 nodes [
  n1 = CollectionParameter() {collection=&FortniteMaterialParameters, parameterName=StormCloudLocation}
  n2 = SphereMask(
   A=ComponentMask(None=@n1) {g=true, r=true}
   B=ComponentMask(None=WorldPosition().XYZ) {g=true, r=true}
   Radius=ScalarParameter() {defaultValue=12000, parameterName=CloudShadowRadius}
   Hardness=ScalarParameter() {defaultValue=0.85, parameterName=CloudShadowHardness}
  )
  n3 = Subtract(
   A=MaterialFunctionCall(Float3=CameraPositionWS()) {materialFunction=&BreakOutFloat3Components}.B
   B=ScalarParameter() {defaultValue=12000, parameterName=ShipShadowFadeHeightBias}
  )
  n4 = ScalarParameter() {defaultValue=7000, parameterName=ShipShadowFadeHeightScale}
  Multiply(A=@n2, B=Saturate(None=Divide(A=@n3, B=@n4)))
  n5 = StaticBool()
  n6 = TextureObject() {texture=&/T_CloudNoise_Artemis}
  n7 = TextureProperty(Tex=@n6)
  n8 = MaterialFunctionCall!(
   PreviewInMatEditor=@n5
   Size=ScalarParameter() {defaultValue=65707, parameterName=BigCloudSize01}
   Speed=ScalarParameter() {defaultValue=0.5, parameterName=BigCloudSpeed01}
   Resolution=@n7
  ) {materialFunction=&/MF_CloudLayerCoord}
  n9 = TextureSample!(UVs=@n8.UVs, Tex=@n6, Level=@n8.UVs, Apply View MipBias) {mipValueMode=TMVM_MipLevel, samplerSource=SSM_Wrap_WorldGroupSettings, texture=&T_CloudNoise_Artemis}
  n10 = LinearInterpolate(A=ScalarParameter() {defaultValue=0.8, parameterName=dsfsdf}, B=ScalarParameter() {defaultValue=20, parameterName=dsfd}, Alpha=@n9.B) {constA=0.5}
  n11 = MaterialFunctionCall(
   PreviewInMatEditor=@n5
   Size=ScalarParameter() {defaultValue=129563.7734, parameterName=BigCloudSize02}
   Speed=ScalarParameter() {defaultValue=0.2667, parameterName=BigCloudSpeed02}
   Resolution=@n7
  ) {materialFunction=&MF_CloudLayerCoord}
  n12 = TextureSample(UVs=@n11.UVs, Tex=@n6, Level=@n11.UVs) {mipValueMode=TMVM_MipLevel, samplerSource=SSM_Wrap_WorldGroupSettings, texture=&T_CloudNoise_Artemis}
  n13 = CameraPositionWS()
  n14 = WorldPosition()
  n15 = Subtract(
   A=MaterialFunctionCall(Vector 3=Subtract(A=@n13, B=@n14.XYZ)) {materialFunction=/Engine/Functions/Engine_MaterialFunctions02/Utility/VectorLength}.V3 Length
   B=ScalarParameter() {defaultValue=5400.1445, parameterName=BlendBias}
  )
  n16 = ScalarParameter() {defaultValue=13126.5527, parameterName=BlendScale}
  n17 = Power(Base=@n12.A)
  n18 = Multiply(A=@n16, B=Constant() {r=1.7589}) {constB=0.5}
  n19 = Divide(A=Clamp(None=@n15, Max=@n15), B=Add(A=@n16, B=Multiply(A=@n17, B=@n18)))
  n20 = LinearInterpolate(
   A=ScalarParameter() {parameterName=FinalClipMin}
   B=ScalarParameter() {defaultValue=1, parameterName=FinalClipMax}
   Alpha=LinearInterpolate(A=@n9.R, B=Multiply(A=@n10, B=@n12.R), Alpha=Saturate(None=@n19))
  )
  Multiply(
   A=LinearInterpolate(
    A=Constant() {r=1}
    B=LinearInterpolate(Alpha=@n2) {constA=1, constB=0.1}
    Alpha=CollectionParameter() {collection=&FortniteMaterialParameters, parameterName=isStormActive}
   )
   B=LinearInterpolate(A=ScalarParameter() {defaultValue=0.45, parameterName=shadowMin}, Alpha=Saturate(None=@n20))
  )
  OneMinus(None=@n12.R)
 ]
}
