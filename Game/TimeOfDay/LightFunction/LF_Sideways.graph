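# Light function material for time-of-day lighting that builds an animated, world-space swirl pattern to modulate light.
# Panning wear noise distorts the coordinates of a separately panning corruption swirl texture, whose alpha is inverted and clamped.
# A static, contrast-adjusted soft-noise mask blends that pattern against full white.
# All three layers use triplanar-style projection blended by vertex normals; band and scene-normal switch nodes feed nothing visible.
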Material /Game/TimeOfDay/LightFunction/LF_Sideways
graph {
 properties {materialDomain=MD_LightFunction, pixelDepthOffsetMode=PDOM_Legacy, usedShadingModels=Default Lit}
 nodes [
  n1 = WorldPosition()
  n2 = VectorParameter() {parameterName=SwirlTiling, defaultValue=<rgba 3000 3000 3000 1>}
  n3 = Time()
  n4 = VectorParameter() {parameterName=WebNoisePanSpeed, defaultValue=<rgba -0.01 -0.01 0.01 1>}
  n5 = Add(
   A=Divide(A=WorldPosition().XYZ, B=VectorParameter() {parameterName=PatternTiling, defaultValue=<rgba 1024 1024 1024 1>}.RGB)
   B=Multiply(A=Time(), B=VectorParameter() {parameterName=FlowPanSpeedAmdDorection, defaultValue=<rgba 0.05 0.05 0.05 0>}.RGB)
  )
  n6 = TextureSample(UVs=ComponentMask(None=@n5) {g=true, r=true}) {samplerType=SAMPLERTYPE_Masks, texture=/Game/Global/Textures/Default/Noise/T_WrapEdgeWear_Noise}
  n7 = TextureSample(UVs=ComponentMask(None=@n5) {b=true, g=true}) {samplerType=SAMPLERTYPE_Masks, texture=&T_WrapEdgeWear_Noise}
  n8 = VertexNormalWS()
  n9 = Abs(None=ComponentMask(None=@n8) {r=true})
  n10 = ComponentMask(None=@n5) {b=true, r=true}
  n11 = ComponentMask(None=@n8) {g=true}
  n12 = Multiply(A=LinearInterpolate(
    A=LinearInterpolate(A=@n6.RGB, B=@n7.RGB, Alpha=@n9)
    B=TextureSample(UVs=@n10) {samplerType=SAMPLERTYPE_Masks, texture=&T_WrapEdgeWear_Noise}.RGB
    Alpha=Abs(None=@n11)
   )) {constB=2}
  n13 = Add(
   A=Add(A=Divide(A=@n1.XYZ, B=@n2.RGB), B=Multiply(A=@n3, B=@n4.RGB))
   B=Multiply(A=Subtract(A=@n12), B=ScalarParameter() {defaultValue=0.01, parameterName=WaveDistortion})
  )
  Add(A=@n13)
  n14 = Constant() {r=1}
  n15 = StaticBool() {value=true}
  n16 = TextureSample!(UVs=ComponentMask(None=@n13) {g=true, r=true}, Tex, Level=@n14, Apply View MipBias=@n15) {
   mipValueMode=TMVM_MipLevel
   samplerType=SAMPLERTYPE_LinearColor
   texture=/Game/Athena/Apollo/Environments/Landscape/Materials/Corruption/T_Corruption_SwirlDetailPack
  }
  n17 = TextureSample(UVs=ComponentMask(None=@n13) {b=true, g=true}, Level=@n14, Apply View MipBias=@n15) {
   mipValueMode=TMVM_MipLevel
   samplerType=SAMPLERTYPE_LinearColor
   texture=&T_Corruption_SwirlDetailPack
  }
  n18 = VertexNormalWS()
  n19 = Abs(None=ComponentMask(None=@n18) {r=true})
  n20 = TextureSample(UVs=ComponentMask(None=@n13) {b=true, r=true}, Level=@n14, Apply View MipBias=@n15) {
   mipValueMode=TMVM_MipLevel
   samplerType=SAMPLERTYPE_LinearColor
   texture=&T_Corruption_SwirlDetailPack
  }
  n21 = ComponentMask(None=@n18) {g=true}
  n22 = ComponentMask(None=LinearInterpolate(A=LinearInterpolate(A=@n16.RGBA, B=@n17.RGBA, Alpha=@n19), B=@n20.RGBA, Alpha=Abs(None=@n21))) {a=true}
  n23 = Clamp(None=OneMinus(None=Multiply(A=@n22) {constB=1.5})) {minDefault=0.1}
  Frac(None=Multiply(A=@n23, B=ScalarParameter() {defaultValue=1, parameterName=BandMultiplier}))
  n24 = Add(A=Divide(A=WorldPosition().XYZ, B=VectorParameter() {parameterName=Mask, defaultValue=<rgba 4096 4096 4096 1>}.RGB))
  n25 = VertexNormalWS()
  n26 = LinearInterpolate(
   A=TextureSample!(UVs=ComponentMask(None=@n24) {g=true, r=true}, Tex, Apply View MipBias) {texture=/Game/Characters/Player/Male/Medium/Bodies/M_MED_Viper/Textures/Glitch/T_SoftNoise}.RGB
   B=TextureSample(UVs=ComponentMask(None=@n24) {b=true, g=true}) {texture=&T_SoftNoise}.RGB
   Alpha=Abs(None=ComponentMask(None=@n25) {r=true})
  )
  n27 = TextureSample(UVs=ComponentMask(None=@n24) {b=true, r=true}) {texture=&T_SoftNoise}
  n28 = Abs(None=ComponentMask(None=@n25) {g=true})
  n29 = Power(
   Base=Multiply(A=LinearInterpolate(A=@n26, B=@n27.RGB, Alpha=@n28), B=ScalarParameter() {defaultValue=2, parameterName=StrMask})
   Exp=ScalarParameter() {defaultValue=2, parameterName=MaskPwr}
  )
  LinearInterpolate(B=@n23, Alpha=ComponentMask(None=Clamp(None=@n29)) {r=true}) {constA=1}
  ShadingPathSwitch(Default=SceneTexture() {sceneTextureId=PPI_WorldNormal}.Color, Mobile=@n8)
  ShadingPathSwitch(Default=SceneTexture() {sceneTextureId=PPI_WorldNormal}.Color, Mobile=@n18)
  ShadingPathSwitch(Default=SceneTexture() {sceneTextureId=PPI_WorldNormal}.Color, Mobile=@n25)
  VectorParameter() {parameterName=SwirlMask, defaultValue=<rgba 1500 1500 1500 1>}
 ]
}
